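# Translucent unlit material for small animated bats on the vampire glider, usable on meshes and particles.
# Emissive color fakes sun diffuse and specular lighting, with switchable fresnel, time-of-day tint and particle color.
# World position offset flaps wings through vertex colors and sine; an optional Pivot Painter mode loops each bat along random arcs with scaling and fades.
# Opacity uses masks with depth fade or smoke erosion driven by dynamic parameters, plus optional near-camera fade.
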Material /Game/Gadgets/Assets/VinderTech_GliderChute/Glider_Vampire/Materials/M_Glider_Bats_Small_anim
graph {
 properties {
  bUsedWithMeshDeformer=true
  bUsedWithMeshParticles=true
  bUsedWithNiagaraMeshParticles=true
  bUsedWithSkeletalMesh=true
  blendMode=BLEND_Translucent
  pixelDepthOffsetMode=PDOM_Legacy
  shadingModel=MSM_Unlit
  twoSided=true
  usedShadingModels=Unlit
 }
 nodes [
  n1 = TextureSample() {samplerType=SAMPLERTYPE_Masks, texture=/Game/Gadgets/Assets/VinderTech_GliderChute/Glider_Vampire/Textures/T_Glider_Bat_M}
  n2 = ScalarParameter() {defaultValue=0.5, parameterName=MorningAndEvening}
  n3 = ScalarParameter() {defaultValue=4, parameterName=Day}
  n4 = ScalarParameter() {defaultValue=0.5, parameterName=Night}
  n5 = ComponentMask(None=CollectionParameter() {collection=&FortniteMaterialParameters, parameterName=SunLightColor}) {b=true, g=true, r=true}
  n6 = ScalarParameter() {defaultValue=0.5, parameterName=DesatReallight}
  n7 = CollectionParameter() {
   collection=&FortniteMaterialParameters
   parameterName=SunlightDirectionalVector
  }
  n8 = ComponentMask(None=@n7) {b=true, g=true, r=true}
  n9 = TextureSample() {samplerType=SAMPLERTYPE_Normal, texture=/Game/Gadgets/Assets/VinderTech_GliderChute/Glider_Vampire/Textures/T_Glider_Bat_N}
  n10 = Transform(None=@n9.RGB)
  n11 = DotProduct(A=@n8, B=@n10)
  n12 = LinearInterpolate(
   A=VectorParameter() {parameterName=FakeFresnell, defaultValue=<rgba 0.0002 0.003 0.01 1>}.RGB
   B=Multiply(A=VectorParameter() {parameterName=FakeFresnell_2, defaultValue=<rgba 0.005 0.0039 0.0035 1>}.RGB, B=Desaturation(None=@n5, Fraction=@n6))
   Alpha=Power(Base=Clamp(None=@n11), Exp=ScalarParameter() {defaultValue=1, parameterName=FakelLightPower})
  )
  n13 = CameraVectorWS()
  n14 = DotProduct(A=@n10, B=Normalize(VectorInput=Add(A=@n8, B=@n13)))
  n15 = Multiply(
   A=VectorParameter() {parameterName=SpecularColor, defaultValue=<rgba 0.01 0.0057 0.0056 1>}.RGB
   B=Power(Base=Saturate(None=@n14), Exp=ScalarParameter() {defaultValue=15.9283, parameterName=FakelSpecPower_1})
  )
  n16 = TextureSample() {texture=/Game/Gadgets/Assets/VinderTech_GliderChute/Glider_Vampire/Textures/T_Glider_Bat_D}
  n17 = Constant2Vector() {g=1, r=1}
  n18 = TwoSidedSign()
  n19 = Transform(None=Multiply(A=AppendVector(A=@n17, B=@n18), B=@n9.RGB))
  n20 = LinearInterpolate(
   A=@n16.RGB
   B=VectorParameter() {parameterName=Color_Fresnel_1, defaultValue=<rgba 0.0138 0.0155 0.0208 1>}.RGB
   Alpha=Fresnel(ExponentIn=ScalarParameter() {defaultValue=1.9571, parameterName=Fresnel_1}, Normal=@n19)
  )
  n21 = Multiply(A=Constant3Vector() {constant=<rgba 0 0 1 0>}, B=@n18)
  n22 = Multiply(
   A=VectorParameter() {parameterName=Color_Fresnel_2, defaultValue=<rgba 0.0001 0.0012 0.005 1>}.RGB
   B=Power(Base=DotProduct(A=@n19, B=@n21), Exp=ScalarParameter() {defaultValue=4.3074, parameterName=Fresnel_2})
  )
  n23 = ScalarParameter() {defaultValue=0.25, parameterName=Brightness}
  n24 = Add(
   A=Multiply(A=ScalarParameter() {defaultValue=0.581, parameterName=BrightnessFakeLight}, B=Add(A=@n12, B=@n15))
   B=StaticSwitchParameter(True=Add(A=@n20, B=@n22), False=Multiply(A=@n16.RGB, B=@n23)) {parameterName=UseFresnel}
  )
  n25 = StaticSwitchParameter(
   True=Multiply(
    A=MaterialFunctionCall(Color_Morning=@n2, Color_Day=@n3, Color_Evening=@n2, Color_Night=@n4) {materialFunction=&Color_Time_of_Day}.Result
    B=@n24
   )
   False=@n24
  ) {parameterName=TimeOfDayControl}
  n26 = Add(A=@n25, B=ParticleColor().RGB)
  MaterialFunctionCall(Emissive=StaticSwitchParameter(True=Multiply(A=@n1.G, B=@n26), False=@n25) {parameterName=AddParticleColorToEmissive}) {materialFunction=&MF_FX_EmissiveAdjustments}
  Saturate(None=Abs(None=@n11))
  n27 = ComponentMask(None=TextureCoordinate()) {r=true}
  n28 = VertexColor()
  n29 = Multiply(
   A=Sign(None=Subtract(A=@n27) {constB=0.5})
   B=LinearInterpolate(
    A=ScalarParameter() {defaultValue=-0.389, parameterName=InwardVector}
    B=ScalarParameter() {defaultValue=-1.0138, parameterName=InwardVector_1}
    Alpha=@n28.G
   )
  )
  n30 = ScalarParameter() {defaultValue=0.134, parameterName=ForwardVector}
  n31 = AppendVector(A=@n29, B=@n30)
  n32 = AppendVector(A=@n30, B=@n29)
  n33 = ScalarParameter() {defaultValue=0.5823, parameterName=UPVector}
  n34 = ScalarParameter() {defaultValue=1.4967, parameterName=UPVector_1}
  n35 = Transform(None=AppendVector(A=StaticSwitchParameter(True=@n31, False=@n32) {parameterName=SwizzleXY}, B=LinearInterpolate(A=@n33, B=@n34, Alpha=@n28.B))) {transformSourceType=TRANSFORMSOURCE_Local}
  n36 = ScalarParameter() {defaultValue=6.4909, parameterName=WindMovementDistance}
  n37 = Time()
  n38 = DynamicParameter() {defaultValue=<rgba 1 1 0.3 1>, paramNames [WingAnimTimeOffset, WingAnimSpeedMultiplier, ErodeAlpha, SmokeLerp]}
  n39 = StaticSwitchParameter(
   True=Add(
    A=@n37
    B=MaterialFunctionCall() {materialFunction=/Engine/Functions/Engine_MaterialFunctions02/PivotPainter/PivotPainter_PerObjectData}.Random Value Per Element
   )
   False=ShadingPathSwitch(Default=Add(A=@n37, B=@n38.WingAnimTimeOffset), Mobile=@n37) {desc=Migrated from FeatureLevelSwitch}
  ) {parameterName=UseAttachedAnimation}
  n40 = Multiply(A=@n28.G, B=ScalarParameter() {defaultValue=0.2, parameterName=WingOffset}) {constB=0.5}
  n41 = ScalarParameter() {defaultValue=4, parameterName=SpeedWings}
  n42 = Multiply(A=@n41, B=@n38.WingAnimSpeedMultiplier)
  n43 = Sine(None=Multiply(A=Add(A=@n39, B=@n40), B=ShadingPathSwitch(Default=@n42, Mobile=@n41) {desc=Migrated from FeatureLevelSwitch}))
  n44 = ScalarParameter() {defaultValue=0.7853, parameterName=WindMovementBias}
  n45 = Multiply(
   A=Multiply(A=Multiply(A=@n35, B=@n36), B=Add(A=@n43, B=@n44))
   B=Add(A=@n28.R, B=ScalarParameter() {defaultValue=0.2286, parameterName=BodyMovement})
  )
  n46 = ScalarParameter() {defaultValue=-228.3521, parameterName=MovementDistance_start}
  n47 = ScalarParameter() {defaultValue=228.3521, parameterName=MovementDistance_End}
  n48 = MaterialFunctionCall() {materialFunction=&PivotPainter_PerObjectData}
  n49 = Add(
   A=Divide(A=Time(), B=ScalarParameter() {defaultValue=2, parameterName=LifeTime, sortPriority=1})
   B=Multiply(A=@n48.Random Value Per Element, B=ScalarParameter() {defaultValue=1692.458, parameterName=TimeOffset})
  )
  n50 = Frac(None=@n49)
  n51 = VectorParameter() {parameterName=Forward, defaultValue=<rgba 0 -1 0 1>}
  n52 = Ceil(None=@n49)
  n53 = Frac(None=Add(A=Multiply(A=@n52) {constB=0.1}, B=@n48.Random Value Per Element)) {desc=Random Selection}
  n54 = LinearInterpolate(
   A=LinearInterpolate(
    A=VectorParameter() {parameterName=MoventVector, defaultValue=<rgba 4 -1 -2 1>}.RGB
    B=VectorParameter() {parameterName=MoventVector_3, defaultValue=<rgba 2 -1 2 1>}.RGB
    Alpha=@n53
   )
   B=LinearInterpolate(
    A=VectorParameter() {parameterName=MoventVector_2, defaultValue=<rgba -4 -1 4 1>}.RGB
    B=VectorParameter() {parameterName=MoventVector_1, defaultValue=<rgba -2 -1 -2 1>}.RGB
    Alpha=@n53
   )
   Alpha=@n53
  )
  n55 = Normalize(VectorInput=Transform(None=LinearInterpolate(A=@n51.RGB, B=@n54, Alpha=@n50)) {transformSourceType=TRANSFORMSOURCE_Local})
  n56 = ScalarParameter() {defaultValue=-0.7769, parameterName=ScaleBats}
  n57 = Clamp(None=Multiply(A=@n50, B=ScalarParameter() {defaultValue=2, parameterName=FadeinSpeed}))
  n58 = OneMinus(None=@n57)
  n59 = WorldPosition()
  n60 = Add(
   A=Multiply(A=LinearInterpolate(A=@n46, B=@n47, Alpha=@n50), B=@n55)
   B=Multiply(A=LinearInterpolate(A=@n56, Alpha=@n58) {constB=0}, B=Subtract(A=@n48.Pivot Position, B=@n59.XYZ))
  )
  n61 = VectorParameter() {parameterName=Right, defaultValue=<rgba 0 -1 0 1>}
  n62 = RotateAboutAxis(
   NormalizedRotationAxis=CrossProduct(A=@n55, B=Transform(None=@n61.RGB) {transformSourceType=TRANSFORMSOURCE_Local})
   RotationAngle=Multiply(A=@n50, B=ScalarParameter() {defaultValue=-0.2, parameterName=MaxRotation})
   PivotPoint=@n48.Pivot Position
   Position=@n59.XYZ
  )
  StaticSwitchParameter(True=Add(A=@n45, B=Add(A=@n60, B=@n62)), False=@n45) {parameterName=UseAttachedAnimation}
  StaticSwitchParameter() {parameterName=UseAttachedAnimation}
  n63 = Multiply(
   A=@n1.G
   B=TextureSample(UVs=TextureCoordinate() {uTiling=3, vTiling=3}) {samplerType=SAMPLERTYPE_Grayscale, texture=&T_Smoke_Trail_Soft}.G
  )
  n64 = DynamicParameter() {defaultValue=<rgba 1 1 0.3 1>, paramNames [WingAnimTimeOffset, WingAnimSpeedMultiplier, ErodeAlpha, SmokeLerp]}
  n65 = DynamicParameter() {defaultValue=<rgba 1 1 0.3 1>, paramNames [WingAnimTimeOffset, WingAnimSpeedMultiplier, ErodeAlpha, SmokeLerp]}
  n66 = Multiply(A=@n1.R, B=@n1.R)
  n67 = OneMinus(None=@n50)
  n68 = Multiply(A=StaticSwitchParameter(True=Multiply(A=@n57, B=@n67), False=ParticleColor().A) {parameterName=UseAttachedAnimation}) {constB=4}
  n69 = Clamp(None=Power(Base=Multiply(A=@n66, B=@n68), Exp=ScalarParameter() {defaultValue=0.6555, parameterName=Contrast}))
  n70 = ScalarParameter() {defaultValue=40, parameterName=DepthFade}
  n71 = StaticSwitchParameter(
   True=MaterialFunctionCall!(Dynamic Param=LinearInterpolate(A=@n1.G, B=@n63, Alpha=@n64.SmokeLerp), Alpha=@n65.ErodeAlpha) {materialFunction=/Game/Effects/Fort_Effects/Materials/_Functions/Erode_Alpha}.Result
   False=ShadingPathSwitch(Default=DepthFade(Opacity=@n69, FadeDistance=@n70), Mobile=@n69) {desc=Migrated from FeatureLevelSwitch}
  ) {parameterName=AddAlphaSmokeErode}
  n72 = ScalarParameter() {parameterName=Near Camera Fade}
  StaticSwitchParameter(
   True=Multiply(A=@n71, B=MaterialFunctionCall(Distance=@n72) {materialFunction=&Near_Camera_Fade}.Result)
   False=@n71
  ) {parameterName=NearFade}
 ]
}
